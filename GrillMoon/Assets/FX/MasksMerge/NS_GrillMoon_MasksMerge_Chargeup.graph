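# Niagara system for the charge-up effect of the GrillMoon masks merge, built from one sprite emitter, Burst_glow.
# The emitter spawns short-lived glow sprites at a steady rate on a roughly two-second loop, offset along the camera axis.
# Curves pop each sprite's size then shrink it, fade alpha in and out, and shift HDR orange toward near black.
# User parameters for skeletal meshes, front-end flags and burst probability are exposed but unreferenced in the modules shown.
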NiagaraSystem /GrillMoon/Assets/FX/MasksMerge/NS_GrillMoon_MasksMerge_Chargeup
dependencies {
 usedDataInterfaces [&NiagaraDataInterfaceCurve, &NiagaraDataInterfaceColorCurve]
 usedDynamicInputs []
 usedModules [&SystemState]
}
emitters {Burst_glow {
  emitterUpdateScript [
   &EmitterState(Life Cycle Mode=::NewEnumerator1, Loop Duration=2.15, UseLoopDelay=false)
   &SpawnRate(SpawnRate=10, Use Spawn Probability=false)
  ]
  particleSpawnScript [
   &InitializeParticle(
    Color=<rgba 3 3 3 1>
    Lifetime=0.5
    Sprite Rotation Mode=::NewEnumerator2
    Sprite Size Mode=::NewEnumerator3
    Uniform Sprite Size=25
    UsePositionOffset=false
    Write Material Random=true
   )
   &CameraOffset(Camera Offset Amount=100)
  ]
  particleUpdateScript [
   &ParticleState()
   &ScaleSpriteSize(Uniform Curve Sprite Scale=DataInterface() {propertyValues {Curve {keys [
        {arriveTangent=-0.4932, interpMode=RCIM_Cubic, leaveTangent=-0.4932, tangentMode=RCTM_User, time=0, value=0.55}
        {arriveTangent=-3.0216, interpMode=RCIM_Cubic, leaveTangent=-3.0216, tangentMode=RCTM_User, time=0.05, value=1.5}
        <at 1 0 RCIM_Cubic>
       ]}}})
   &ScaleColor(
    Scale Alpha=&Multiply_Float(
     A=&FloatFromCurve(FloatCurve=DataInterface() {propertyValues {Curve {keys [<at 0 0>,<at 0.1 1>,<at 0.4 1>,<at 1 0>]}}}, Scale Curve=0.5)
     B=1
    )
    Scale RGB=&MakeVectorFromLinearColorRGB(Color=&ColorFromCurve(ColorCurve=DataInterface() {propertyValues {
        AlphaCurve {keys []}
        BlueCurve {keys [
          {arriveTangent=-0.0154, interpMode=RCIM_Cubic, time=0.1561, value=0}
          {interpMode=RCIM_Cubic, leaveTangent=-0.0154, time=0.9205, value=0.0053}
         ]}
        GreenCurve {keys [
          {arriveTangent=-0.0154, interpMode=RCIM_Cubic, time=0.1561, value=0.9923}
          {interpMode=RCIM_Cubic, leaveTangent=-0.0154, time=0.9205, value=0}
         ]}
        RedCurve {keys [
          {arriveTangent=-0.0154, interpMode=RCIM_Cubic, time=0.1561, value=4}
          {interpMode=RCIM_Cubic, leaveTangent=-0.0154, time=0.9205, value=0.0365}
         ]}
       }}))
    ScaleA=true
    ScaleRGB=true
   )
  ]
  rendererClasses [&NiagaraSpriteRendererProperties]
  renderers [{rendererClass=&NiagaraSpriteRendererProperties}]
 }}
summary {
 systemName=NS_GrillMoon_MasksMerge_Chargeup
 emitters [{emitterName=Burst_glow, rendererClasses [&NiagaraSpriteRendererProperties]}]
 userVariables [
  User.bIsFrontEnd {type=NiagaraBool}
  User.bIsFrontEndPreview {type=NiagaraBool}
  User.BurstEffectSpawnProbability {type=NiagaraFloat, defaultValue=1, description=""}
  User.SkeletalMesh {
   type=NiagaraDataInterfaceSkeletalMesh
   defaultValue {
    struct=NiagaraExt_VariableValue_DataInterface
    value {dataInterface=&self:NiagaraDataInterfaceSkeletalMesh_0}
   }
  }
  User.SkelMesh {
   type=NiagaraDataInterfaceSkeletalMesh
   defaultValue {
    struct=NiagaraExt_VariableValue_DataInterface
    value {dataInterface=&self:NiagaraDataInterfaceSkeletalMesh_1}
   }
  }
 ]
}
